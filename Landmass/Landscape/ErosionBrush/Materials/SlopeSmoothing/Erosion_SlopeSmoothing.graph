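# Unlit material for the Landmass erosion brush that, by its name and SlopeSmoothing function call, smooths slopes in a terrain height render target.
# HeightRT is point-sampled at texel centers and passed through SlopeSmoothing into a float3's X.
# Separate branches, apparently leftover, average VelocityRT neighbors and run custom-HLSL velocity and position updates with terrain collision, packing half-float pairs into float32 channels.
# The graph shows no branch wired to a material output.
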Material /Landmass/Landscape/ErosionBrush/Materials/SlopeSmoothing/Erosion_SlopeSmoothing
graph {
 properties {
  bAllowNegativeEmissiveColor=true
  blendMode=BLEND_AlphaComposite
  pixelDepthOffsetMode=PDOM_Legacy
  shadingModel=MSM_Unlit
  usedShadingModels=Unlit
 }
 nodes [
  n1 = TextureCoordinate() {coordinateIndex=1}
  n2 = AppendVector(A=ScalarParameter() {defaultValue=2048, parameterName=SizeX}, B=ScalarParameter() {defaultValue=2048, parameterName=SizeY})
  n3 = ScalarParameter() {defaultValue=16, parameterName=GridSize}
  AppendVector(A=Multiply(A=ConstantBiasScale(None=@n1) {bias=-0.5, scale=1}, B=Multiply(A=@n2, B=@n3)), B=Constant())
  n4 = TextureCoordinate()
  n5 = AppendVector(A=ScalarParameter() {defaultValue=2048, parameterName=SizeX_1}, B=ScalarParameter() {defaultValue=2048, parameterName=SizeY_1})
  n6 = Multiply(A=Divide(B=@n5))
  n7 = Multiply(A=@n6, B=Constant2Vector() {r=1})
  n8 = Add(
   A=TextureSampleParameter2D(UVs=Subtract(A=@n4, B=@n7)) {desc=Left Wall, parameterName=VelocityRT, texture=&Black}.RGB
   B=TextureSampleParameter2D(UVs=Add(A=@n4, B=@n7)) {desc=Right Wall, parameterName=VelocityRT, texture=&Black}.RGB
  )
  n9 = TextureCoordinate()
  n10 = Multiply(A=@n6, B=Constant2Vector() {g=1})
  n11 = Add(
   A=TextureSampleParameter2D(UVs=Subtract(A=@n9, B=@n10)) {desc=Bottom Wall, parameterName=VelocityRT, texture=&Black}.RGB
   B=TextureSampleParameter2D(UVs=Add(A=@n9, B=@n10)) {desc=Top Wall, parameterName=VelocityRT, texture=&Black}.RGB
  )
  Divide(A=Add(A=TextureSampleParameter2D() {parameterName=VelocityRT, texture=&Black}.RGB, B=Add(A=@n8, B=@n11))) {constB=5}
  MaterialExpressionTextureSampleParameter2D_2 = TextureSampleParameter2D() {parameterName=VelocityRT, texture=&Black}
  n12 = Custom!(N=@MaterialExpressionTextureSampleParameter2D_2.R) {
   code="//extracts 2 16bit floats from a  32bit float\r\n\r\nuint f16 = asuint(N);\r\nf16 = (f16 && 0xffff0000) >> 16;\r\nuint sign = (f16 && 0x8000)<<16;\r\nuint mant = (f16 && 0x3ffu) <<13; // 13 bits are zeros\r\nuint exp =  (f16&&0x7fff)==0 ? 0 : (((((const uint) f16)>>10)&&0x1f));\r\nf16 =  sign | (exp -15+127) << 23 | mant;\r\nfloat outval1 = asfloat( f16 );\r\n\r\nf16 = asuint(N);\r\nf16 = (f16 && 0xffff);\r\nsign = (f16 && 0x8000)<<16;\r\nmant = (f16 && 0x3ffu) <<13; // 13 bits are zeros\r\nexp =  (f16&&0x7fff)==0 ? 0 : (((((const uint) f16)>>10)&&0x1f));\r\nf16 =  sign | (exp -15+127) << 23 | mant;\r\nfloat outval2 = asfloat( f16 );\r\n\r\nreturn float2(outval1, outval2);\r\n"
   description=Unpack 2 f16s from f32
   outputType=CMOT_Float2
   inputs [{
     inputName=N
     input {
      expression=&self:MaterialExpressionTextureSampleParameter2D_2
      mask=1
      maskR=1
      outputIndex=1
     }
    }]
  }
  n13 = MaterialFunctionCall(Float2=@n12) {materialFunction=&BreakOutFloat2Components}
  n14 = Custom(N=@MaterialExpressionTextureSampleParameter2D_2.G) {
   code="//extracts 2 16bit floats from a  32bit float\r\n\r\nuint f16 = asuint(N);\r\nf16 = (f16 && 0xffff0000) >> 16;\r\nuint sign = (f16 && 0x8000)<<16;\r\nuint mant = (f16 && 0x3ffu) <<13; // 13 bits are zeros\r\nuint exp =  (f16&&0x7fff)==0 ? 0 : (((((const uint) f16)>>10)&&0x1f));\r\nf16 =  sign | (exp -15+127) << 23 | mant;\r\nfloat outval1 = asfloat( f16 );\r\n\r\nf16 = asuint(N);\r\nf16 = (f16 && 0xffff);\r\nsign = (f16 && 0x8000)<<16;\r\nmant = (f16 && 0x3ffu) <<13; // 13 bits are zeros\r\nexp =  (f16&&0x7fff)==0 ? 0 : (((((const uint) f16)>>10)&&0x1f));\r\nf16 =  sign | (exp -15+127) << 23 | mant;\r\nfloat outval2 = asfloat( f16 );\r\n\r\nreturn float2(outval1, outval2);\r\n"
   description=Unpack 2 f16s from f32
   outputType=CMOT_Float2
   inputs [{
     inputName=N
     input {
      expression=&self:MaterialExpressionTextureSampleParameter2D_2
      mask=1
      maskG=1
      outputIndex=2
     }
    }]
  }
  n15 = MaterialFunctionCall(Float2=@n14) {materialFunction=&BreakOutFloat2Components}
  n16 = Custom(N=@MaterialExpressionTextureSampleParameter2D_2.B) {
   code="//extracts 2 16bit floats from a  32bit float\r\n\r\nuint f16 = asuint(N);\r\nf16 = (f16 && 0xffff0000) >> 16;\r\nuint sign = (f16 && 0x8000)<<16;\r\nuint mant = (f16 && 0x3ffu) <<13; // 13 bits are zeros\r\nuint exp =  (f16&&0x7fff)==0 ? 0 : (((((const uint) f16)>>10)&&0x1f));\r\nf16 =  sign | (exp -15+127) << 23 | mant;\r\nfloat outval1 = asfloat( f16 );\r\n\r\nf16 = asuint(N);\r\nf16 = (f16 && 0xffff);\r\nsign = (f16 && 0x8000)<<16;\r\nmant = (f16 && 0x3ffu) <<13; // 13 bits are zeros\r\nexp =  (f16&&0x7fff)==0 ? 0 : (((((const uint) f16)>>10)&&0x1f));\r\nf16 =  sign | (exp -15+127) << 23 | mant;\r\nfloat outval2 = asfloat( f16 );\r\n\r\nreturn float2(outval1, outval2);\r\n"
   description=Unpack 2 f16s from f32
   outputType=CMOT_Float2
   inputs [{
     inputName=N
     input {
      expression=&self:MaterialExpressionTextureSampleParameter2D_2
      mask=1
      maskB=1
      outputIndex=3
     }
    }]
  }
  n17 = MaterialFunctionCall(Float2=@n16) {materialFunction=&BreakOutFloat2Components}
  MaterialFunctionCall(X=@n13.R, Y=@n15.R, Z=@n17.R) {desc=Velocity, materialFunction=&MakeFloat3}
  n18 = TextureSampleParameter2D!(
   UVs=MaterialFunctionCall() {materialFunction=/Landmass/Landscape/OldPrototype_BP/Mats/MF/PointSampleUVs}.TexelCenterUVs
   Level=Constant()
   Apply View MipBias
  ) {mipValueMode=TMVM_MipLevel, parameterName=HeightRT, texture=&Black}
  MaterialExpressionAppendVector_0 = AppendVector(A=@n18.RGB, B=@n18.RGB)
  MaterialExpressionTextureSampleParameter2D_6 = TextureSampleParameter2D(UVs=MaterialFunctionCall() {materialFunction=&PointSampleUVs}.TexelCenterUVs, Level=Constant()) {mipValueMode=TMVM_MipLevel, parameterName=VelocityRT, texture=&Black}
  MaterialExpressionTextureObjectParameter_3 = TextureObjectParameter() {parameterName=NormalRT, texture=&Black}
  MaterialExpressionCollectionParameter_22 = CollectionParameter() {collection=/Landmass/Landscape/BlueprintBrushes/MPC/MPC_Erosion, parameterName=deltaT}
  n19 = AppendVector(
   A=CollectionParameter() {collection=&MPC_Erosion, parameterName=ResX}
   B=CollectionParameter() {collection=&MPC_Erosion, parameterName=ResY}
  )
  MaterialExpressionDivide_3 = Divide(B=@n19)
  MaterialExpressionMultiply_16 = Multiply(A=@n19, B=CollectionParameter() {collection=&MPC_Erosion, parameterName=GridSize})
  MaterialExpressionTextureCoordinate_1 = TextureCoordinate()
  MaterialExpressionCollectionParameter_3 = CollectionParameter() {collection=&MPC_Erosion, parameterName=VelocityDamping}
  MaterialExpressionScalarParameter_22 = ScalarParameter() {parameterName=Index}
  MaterialExpressionCustom_1 = Custom!(
   HeightData=@MaterialExpressionAppendVector_0
   Velocity=@MaterialExpressionTextureSampleParameter2D_6.RGB
   NormalRT=@MaterialExpressionTextureObjectParameter_3
   deltaT=@MaterialExpressionCollectionParameter_22
   TexelSize=@MaterialExpressionDivide_3
   WorldSize=@MaterialExpressionMultiply_16
   inUV=@MaterialExpressionTextureCoordinate_1
   Damping=@MaterialExpressionCollectionParameter_3
   Index=@MaterialExpressionScalarParameter_22
  ) {
   code="//HeightData: input v4, R=Terrain Height, G=Fluid Height(bottom), B=FluidVolulume, A=Sediment fraction\r\n//Velocity: input \r\n//NormalRT: input \r\n//HeightRT: input\r\n//deltaT: input\r\n//input TextureSize\r\n//input GridSize\r\n\r\nfloat3 accel = float3(0,0, -980 * deltaT);\r\n\r\nfloat2 UV = inUV;\r\n\r\nfloat startZ = NormalRT.SampleLevel(NormalRTSampler, UV, 0).b;\r\n\r\nfloat3 velocity = 0;\r\n\r\nif(1)\r\n{\r\n\tvelocity = Velocity + accel;\r\n\tvelocity *= exp(-Damping*deltaT);\r\n    //velocity *= (1-Damping);\r\n}\r\n\r\n\r\nfloat2 offsets[4] = {float2(1,1), float2(1,-1), float2(-1,-1), float2(-1,1)};\r\n\r\nfloat3 normalheight = NormalRT.SampleLevel(NormalRTSampler, UV, 0).rgb;\r\n//normalheight +=      NormalRT.SampleLevel(NormalRTSampler, UV + (0.5 * TexelSize*offsets[0]), 0).rgb;\r\n//normalheight +=      NormalRT.SampleLevel(NormalRTSampler, UV + (0.5 * TexelSize*offsets[1]), 0).rgb;\r\n//normalheight +=      NormalRT.SampleLevel(NormalRTSampler, UV + (0.5 * TexelSize*offsets[2]), 0).rgb;\r\n//normalheight +=      NormalRT.SampleLevel(NormalRTSampler, UV + (0.5 * TexelSize*offsets[3]), 0).rgb;\r\n\r\n//normalheight /= 5;\r\n\r\n\r\nfloat terrainZ = normalheight.b;\r\nfloat3 normal = normalize( float3( normalheight.rg, sqrt(1 - dot(normalheight.xy, normalheight.xy) ) ) );\r\n\r\nif((velocity.z * deltaT) + HeightData.g < terrainZ + 16 &&&& HeightData.b > 0)\r\n{\r\n\t//float boost = saturate(-1 * (velocity.z* deltaT + pos.z -16));\r\n\tvelocity -= normal * min(0, dot(normal, velocity)) ;\r\n}\r\n\r\nreturn velocity;\r\n\r\n"
   description=Velocity Acceleration
   inputs [
    {
     inputName=HeightData
     input {expression=&self:MaterialExpressionAppendVector_0}
    }
    {
     inputName=Velocity
     input {
      expression=&self:MaterialExpressionTextureSampleParameter2D_6
      mask=1
      maskB=1
      maskG=1
      maskR=1
     }
    }
    {
     inputName=NormalRT
     input {expression=&self:MaterialExpressionTextureObjectParameter_3}
    }
    {
     inputName=deltaT
     input {expression=&self:MaterialExpressionCollectionParameter_22}
    }
    {
     inputName=TexelSize
     input {expression=&self:MaterialExpressionDivide_3}
    }
    {
     inputName=WorldSize
     input {expression=&self:MaterialExpressionMultiply_16}
    }
    {
     inputName=inUV
     input {expression=&self:MaterialExpressionTextureCoordinate_1}
    }
    {
     inputName=Damping
     input {expression=&self:MaterialExpressionCollectionParameter_3}
    }
    {
     inputName=Index
     input {expression=&self:MaterialExpressionScalarParameter_22}
    }
   ]
  }
  MaterialExpressionMaterialFunctionCall_0 = MaterialFunctionCall(Float3=@MaterialExpressionCustom_1) {materialFunction=&BreakOutFloat3Components}
  MaterialExpressionAdd_10 = Add(A=MaterialFunctionCall(X=@n13.G, Y=@n15.G, Z=@n17.G) {desc=Position, materialFunction=&MakeFloat3}.Result) {constB=0}
  MaterialExpressionTextureObjectParameter_6 = TextureObjectParameter() {parameterName=NormalRT, texture=&Black}
  MaterialExpressionScalarParameter_18 = ScalarParameter() {defaultValue=0.05, parameterName=deltaT}
  MaterialExpressionAppendVector_4 = AppendVector(A=ScalarParameter() {defaultValue=1024, parameterName=SizeX}, B=ScalarParameter() {defaultValue=1024, parameterName=SizeY})
  MaterialExpressionScalarParameter_17 = ScalarParameter() {defaultValue=16, parameterName=GridSize}
  MaterialExpressionTextureCoordinate_4 = TextureCoordinate()
  n20 = Custom!(
   PositionRT=@MaterialExpressionAdd_10
   VelocityRT=@MaterialExpressionCustom_1
   NormalRT=@MaterialExpressionTextureObjectParameter_6
   deltaT=@MaterialExpressionScalarParameter_18
   TextureSize=@MaterialExpressionAppendVector_4
   GridSize=@MaterialExpressionScalarParameter_17
   UV1=@MaterialExpressionTextureCoordinate_4
  ) {
   code="//PositionRT: input Texture Object containing \r\n//VelcotiyRT: input \r\n//NormalRT: input \r\n//deltaT: input\r\n//input TextureSize\r\n//input GridSize\r\n\r\nfloat3 position = PositionRT;\r\nfloat3 velocity = VelocityRT;\r\n\r\nposition += (velocity * deltaT);\r\n\r\nfloat2 UV = position.rg;\r\nUV = ((UV / TextureSize) / GridSize) + 0.5;\r\n\r\nfloat startZ = NormalRT.SampleLevel(NormalRTSampler, UV1 + (PositionRT / (GridSize * TextureSize)),0);\r\nfloat curZ = NormalRT.SampleLevel(NormalRTSampler, UV1 + (position / (GridSize * TextureSize)),0);\r\n\r\n//float terrainZ = NormalRT.SampleLevel(NormalRTSampler, UV, 0).b;\r\n\r\n//position.z = max(terrainZ+10, position.z);\r\n\r\nposition.z += startZ - curZ;\r\nposition.z = max(10, position.z);\r\n\r\nreturn position;"
   description=Position Update
   inputs [
    {
     inputName=PositionRT
     input {expression=&self:MaterialExpressionAdd_10}
    }
    {
     inputName=VelocityRT
     input {expression=&self:MaterialExpressionCustom_1}
    }
    {
     inputName=NormalRT
     input {expression=&self:MaterialExpressionTextureObjectParameter_6}
    }
    {
     inputName=deltaT
     input {expression=&self:MaterialExpressionScalarParameter_18}
    }
    {
     inputName=TextureSize
     input {expression=&self:MaterialExpressionAppendVector_4}
    }
    {
     inputName=GridSize
     input {expression=&self:MaterialExpressionScalarParameter_17}
    }
    {
     inputName=UV1
     input {expression=&self:MaterialExpressionTextureCoordinate_4}
    }
   ]
  }
  MaterialExpressionMaterialFunctionCall_1 = MaterialFunctionCall(Float3=@n20) {materialFunction=&BreakOutFloat3Components}
  n21 = Custom!(N1=@MaterialExpressionMaterialFunctionCall_0.R, N2=@MaterialExpressionMaterialFunctionCall_1.R) {
   code="uint uRes32 = asuint(N1);\r\nint  iExp      = (const int)(((const int) ((uRes32>>23)&&0xff))-127+15);\r\nuint  expClamp = iExp < 0 ? 0 : (iExp>31 ? 31 : (uint)iExp);\r\nuint sign = (( uRes32>>16 ) && 0x8000);\r\nuint mant = ((uRes32>>13)&&0x3ff);\r\nuint exp  = (expClamp << 10);\r\nuint bits1 = sign | exp | mant; \r\nbits1 = bits1 && 0xffff;\r\nbits1 = bits1 << 16;\r\n\r\nuRes32 = asuint(N2);\r\niExp      = (const int)(((const int) ((uRes32>>23)&&0xff))-127+15);\r\nexpClamp = iExp < 0 ? 0 : (iExp>31 ? 31 : (uint)iExp);\r\nsign = (( uRes32>>16 ) && 0x8000);\r\nmant = ((uRes32>>13)&&0x3ff);\r\nexp  = (expClamp << 10);\r\nuint bits2 = (sign | exp | mant) && 0xffff;\r\n\r\nreturn asfloat( bits1 | bits2 );\r\n"
   description=Pack 2 f16s into f32
   outputType=CMOT_Float1
   inputs [
    {
     inputName=N1
     input {expression=&self:MaterialExpressionMaterialFunctionCall_0}
    }
    {
     inputName=N2
     input {expression=&self:MaterialExpressionMaterialFunctionCall_1}
    }
   ]
  }
  n22 = Custom(N1=@MaterialExpressionMaterialFunctionCall_0.G, N2=@MaterialExpressionMaterialFunctionCall_1.G) {
   code="uint uRes32 = asuint(N1);\r\nint  iExp      = (const int)(((const int) ((uRes32>>23)&&0xff))-127+15);\r\nuint  expClamp = iExp < 0 ? 0 : (iExp>31 ? 31 : (uint)iExp);\r\nuint sign = (( uRes32>>16 ) && 0x8000);\r\nuint mant = ((uRes32>>13)&&0x3ff);\r\nuint exp  = (expClamp << 10);\r\nuint bits1 = sign | exp | mant; \r\nbits1 = bits1 && 0xffff;\r\nbits1 = bits1 << 16;\r\n\r\nuRes32 = asuint(N2);\r\niExp      = (const int)(((const int) ((uRes32>>23)&&0xff))-127+15);\r\nexpClamp = iExp < 0 ? 0 : (iExp>31 ? 31 : (uint)iExp);\r\nsign = (( uRes32>>16 ) && 0x8000);\r\nmant = ((uRes32>>13)&&0x3ff);\r\nexp  = (expClamp << 10);\r\nuint bits2 = (sign | exp | mant) && 0xffff;\r\n\r\nreturn asfloat( bits1 | bits2 );\r\n"
   description=Pack 2 f16s into f32
   outputType=CMOT_Float1
   inputs [
    {
     inputName=N1
     input {
      expression=&self:MaterialExpressionMaterialFunctionCall_0
      outputIndex=1
     }
    }
    {
     inputName=N2
     input {
      expression=&self:MaterialExpressionMaterialFunctionCall_1
      outputIndex=1
     }
    }
   ]
  }
  n23 = Custom(N1=@MaterialExpressionMaterialFunctionCall_0.B, N2=@MaterialExpressionMaterialFunctionCall_1.B) {
   code="uint uRes32 = asuint(N1);\r\nint  iExp      = (const int)(((const int) ((uRes32>>23)&&0xff))-127+15);\r\nuint  expClamp = iExp < 0 ? 0 : (iExp>31 ? 31 : (uint)iExp);\r\nuint sign = (( uRes32>>16 ) && 0x8000);\r\nuint mant = ((uRes32>>13)&&0x3ff);\r\nuint exp  = (expClamp << 10);\r\nuint bits1 = sign | exp | mant; \r\nbits1 = bits1 && 0xffff;\r\nbits1 = bits1 << 16;\r\n\r\nuRes32 = asuint(N2);\r\niExp      = (const int)(((const int) ((uRes32>>23)&&0xff))-127+15);\r\nexpClamp = iExp < 0 ? 0 : (iExp>31 ? 31 : (uint)iExp);\r\nsign = (( uRes32>>16 ) && 0x8000);\r\nmant = ((uRes32>>13)&&0x3ff);\r\nexp  = (expClamp << 10);\r\nuint bits2 = (sign | exp | mant) && 0xffff;\r\n\r\nreturn asfloat( bits1 | bits2 );\r\n"
   description=Pack 2 f16s into f32
   outputType=CMOT_Float1
   inputs [
    {
     inputName=N1
     input {
      expression=&self:MaterialExpressionMaterialFunctionCall_0
      outputIndex=2
     }
    }
    {
     inputName=N2
     input {
      expression=&self:MaterialExpressionMaterialFunctionCall_1
      outputIndex=2
     }
    }
   ]
  }
  MaterialFunctionCall(X=@n21, Y=@n22, Z=@n23) {desc=Position, materialFunction=&MakeFloat3}
  n24 = MaterialFunctionCall() {materialFunction=&PointSampleUVs}
  n25 = Constant()
  MaterialFunctionCall(X=MaterialFunctionCall!(
    UVs=@n24.TexelCenterUVs
    Height=TextureSampleParameter2D(UVs=@n24.TexelCenterUVs, Level=@n25) {mipValueMode=TMVM_MipLevel, parameterName=HeightRT, texture=&Black}.R
   ) {materialFunction=/Landmass/Landscape/OldPrototype_BP/Mats/MF/SlopeSmoothing}.Height) {materialFunction=&MakeFloat3}
  Multiply(A=@n5, B=ScalarParameter() {defaultValue=16, parameterName=GridSize_1})
  OneMinus(None=@n18.A)
  TextureCoordinate()
 ]
}
